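# Physics asset for the F_MED_Donut_Dish female medium player body cosmetic, covering its left, right and thigh holsters.
# Each holster bone has a box body, paired with a tiny capsule body on its matching MSTR bone.
# Three constraints tie each holster to its MSTR bone with limited swing and tighter twist, restricting how far it can move relative to that anchor.
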PhysicsAsset /BRCosmetics/Characters/Player/Female/Medium/Bodies/F_MED_Donut_Dish/Meshes/F_MED_Donut_Dish_Physics
bodies {
 dyn_holster_l [{shapeType=Box, extentX=17.8984, extentY=12.9513, extentZ=4.0874, center=<-1.8204 -7.8914 -3.4754>}]
 dyn_holster_l_MSTR [{shapeType=Capsule, length=0.505, radius=0.505}]
 dyn_holster_r [{shapeType=Box, extentX=17.8984, extentY=12.9513, extentZ=4.0874, center=<-3.8059 -8.788 2.4649>, rotation=<rot 0 4.4375 0>}]
 dyn_holster_r_MSTR [{shapeType=Capsule, length=0.505, radius=0.505}]
 dyn_holster_thigh [{shapeType=Box, extentX=23.7106, extentY=14.4483, extentZ=4.6732, center=<-2.4663 0.0628 1.7444>}]
 dyn_holster_thigh_MSTR [{shapeType=Capsule, length=0.505, radius=0.505}]
}
constraints [
 {bone1Name=dyn_holster_l, bone2Name=dyn_holster_l_MSTR, swing1LimitDegrees=20, swing2LimitDegrees=20, twistLimitDegrees=5}
 {bone1Name=dyn_holster_r, bone2Name=dyn_holster_r_MSTR, swing1LimitDegrees=20, swing2LimitDegrees=20, twistLimitDegrees=5}
 {bone1Name=dyn_holster_thigh, bone2Name=dyn_holster_thigh_MSTR, swing1LimitDegrees=20, swing2LimitDegrees=20, twistLimitDegrees=5}
]
