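# Physics asset for the hair of the Donut_Dish female medium character, simulating back, side, front and bang strands.
# Capsule chains hang from the head directly or through MSTR root bones, with limited swing and tight or zero twist.
# Free-rotation cross-constraints link neighboring back and right strands so they move together.
# Head, neck, clavicle and right upper arm bodies, the arm adding a box, provide shapes for hair collision.
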PhysicsAsset /BRCosmetics/Characters/Player/Female/Medium/Bodies/F_MED_Donut_Dish/Meshes/Parts/F_MED_Donut_Dish_FaceAcc_Hair_Physics
bodies {
 clavicle_l [{shapeType=Capsule, length=10.1061, radius=4.9227, center=<3.5714 -1.623 0.2881>, rotation=<rot 90 0 0>}]
 clavicle_r [{shapeType=Capsule, length=10.2775, radius=6.3696, center=<-4.1755 1.2428 1.7045>, rotation=<rot 90 0 0>}]
 dyn_hair_b_01 [{shapeType=Capsule, length=1, radius=2, center=<1 0 0>, rotation=<rot 90 0 0>}]
 dyn_hair_b_02 [{shapeType=Capsule, length=1, radius=2, center=<1 0 0>, rotation=<rot 90 0 0>}]
 dyn_hair_b_03 [{shapeType=Capsule, length=1, radius=2, center=<1 0 0>, rotation=<rot 90 0 0>}]
 dyn_hair_b_04 [{shapeType=Capsule, length=1.2445, radius=1.7555, center=<1 0 0>, rotation=<rot 90 0 0>}]
 dyn_hair_b_05 [{shapeType=Capsule, length=1.614, radius=1.636, center=<1 0 0>, rotation=<rot 64.438 0 0>}]
 dyn_hair_b_MSTR [{shapeType=Capsule, length=0.25, radius=0.25}]
 dyn_hair_bang_01 [{shapeType=Capsule, length=1.5131, radius=0.5029, center=<0.9406 0.01 -0.0165>, rotation=<rot 0 96.9026 -89.3906>}]
 dyn_hair_bang_02 [{shapeType=Capsule, length=0.5216, radius=0.4686, center=<0.7674 0.0788 0.0159>, rotation=<rot 0 -111.5866 -60.6834>}]
 dyn_hair_bang_03 [{shapeType=Capsule, length=1.3139, radius=0.3258, center=<0.6927 -0.3204 0.3073>, rotation=<rot 0 70.9195 -74.9158>}]
 dyn_hair_bl_01 [{shapeType=Capsule, length=1, radius=2, center=<1 0 0>, rotation=<rot 90 0 0>}]
 dyn_hair_bl_02 [{shapeType=Capsule, length=1, radius=2, center=<1 0 0>, rotation=<rot 90 0 0>}]
 dyn_hair_bl_03 [{shapeType=Capsule, length=1, radius=2, center=<1 0 0>, rotation=<rot 90 0 0>}]
 dyn_hair_bl_04 [{shapeType=Capsule, length=1.1282, radius=2.0638, center=<1 0 0>, rotation=<rot 90 0 0>}]
 dyn_hair_blm_01 [{shapeType=Capsule, length=1, radius=2, center=<1 0 0>, rotation=<rot 90 0 0>}]
 dyn_hair_blm_02 [{shapeType=Capsule, length=1, radius=2, center=<1 0 0>, rotation=<rot 90 0 0>}]
 dyn_hair_blm_03 [{shapeType=Capsule, length=1, radius=2, center=<1 0 0>, rotation=<rot 90 0 0>}]
 dyn_hair_blm_04 [{shapeType=Capsule, length=1.1282, radius=2.0638, center=<1 0 0>, rotation=<rot 90 0 0>}]
 dyn_hair_blm_MSTR [{shapeType=Capsule, length=0.505, radius=0.505, rotation=<rot 0 -90 0>}]
 dyn_hair_bm_01 [{shapeType=Capsule, length=1, radius=2, center=<1 0 0>, rotation=<rot 90 0 0>}]
 dyn_hair_bm_02 [{shapeType=Capsule, length=1, radius=2, center=<1 0 0>, rotation=<rot 90 0 0>}]
 dyn_hair_bm_03 [{shapeType=Capsule, length=1, radius=2, center=<1 0 0>, rotation=<rot 90 0 0>}]
 dyn_hair_bm_04 [{shapeType=Capsule, length=0.7453, radius=2.1292, center=<1 0 0>, rotation=<rot 90 0 0>}]
 dyn_hair_bm_05 [{shapeType=Capsule, length=0.5304, radius=1.9982, center=<0.1 0 0>, rotation=<rot 90 0 0>}]
 dyn_hair_bm_MSTR [{shapeType=Capsule, length=0.25, radius=0.25}]
 dyn_hair_f_01 [{shapeType=Capsule, length=0.505, radius=0.505, rotation=<rot 0 -90 0>}]
 dyn_hair_fl_01 [{shapeType=Capsule, length=1, radius=2, center=<1 0 0>, rotation=<rot 90 0 0>}]
 dyn_hair_fl_02 [{shapeType=Capsule, length=1, radius=2, center=<1 0 0>, rotation=<rot 90 0 0>}]
 dyn_hair_fl_03 [{shapeType=Capsule, length=1, radius=2, center=<1 0 0>, rotation=<rot 90 0 0>}]
 dyn_hair_fl_04 [{shapeType=Capsule, length=1.2398, radius=1.7602, center=<1 0 0>, rotation=<rot 90 0 0>}]
 dyn_hair_fl_05 [{shapeType=Capsule, length=1.4616, radius=1.5384, center=<1.5897 -0.8898 0>, rotation=<rot 90 90 115.8411>}]
 dyn_hair_fl_MSTR [{shapeType=Capsule, length=0.5, radius=0.5}]
 dyn_hair_l_01 [{shapeType=Capsule, length=1, radius=2, center=<1 0 0>, rotation=<rot 90 10.6197 13.483>}]
 dyn_hair_l_02 [{shapeType=Capsule, length=1, radius=2, center=<1 0 0>, rotation=<rot 90 0 0>}]
 dyn_hair_l_03 [{shapeType=Capsule, length=1, radius=2, center=<1 0 0>, rotation=<rot 90 0 0>}]
 dyn_hair_l_04 [{shapeType=Capsule, length=1.8592, radius=1.8113, center=<1 0 0>, rotation=<rot 90 0 0>}]
 dyn_hair_l_05 [{shapeType=Capsule, length=1.6696, radius=1.3304, center=<1 0 0>, rotation=<rot 90 0 14.8664>}]
 dyn_hair_l_MSTR [{shapeType=Capsule, length=0.5, radius=0.5}]
 dyn_hair_r_01 [{shapeType=Capsule, length=1.4692, radius=1.8897, center=<1 0 0>, rotation=<rot 90 0 0>}]
 dyn_hair_r_02 [{shapeType=Capsule, length=1, radius=2, center=<1 0 0>, rotation=<rot 90 0 0>}]
 dyn_hair_r_03 [{shapeType=Capsule, length=1, radius=2, center=<1 0 0>, rotation=<rot 90 0 0>}]
 dyn_hair_r_04 [{shapeType=Capsule, length=1.1538, radius=1.8462, center=<1 0 0>, rotation=<rot 90 0 0>}]
 dyn_hair_r_05 [{shapeType=Capsule, length=1.441, radius=1.559, center=<1 0.0044 -0.1109>, rotation=<rot 90 0 0>}]
 dyn_hair_r_MSTR [{shapeType=Capsule, length=0.25, radius=0.25}]
 head [{shapeType=Capsule, length=6.3478, radius=6.3107, center=<2.4689 1.2289 -0.9352>, rotation=<rot 90 0 0>}]
 neck_01 [{shapeType=Capsule, length=4.276, radius=5.8574, center=<0 1.2736 0>, rotation=<rot 90 7.125 14.0627>}]
 upperarm_r [
  {shapeType=Capsule, length=10.9065, radius=4.7519, center=<-5.4621 0.6891 0>, rotation=<rot 90 0 0>}
  {shapeType=Box, extentX=16.0929, extentY=15.5015, extentZ=2.3064, center=<-5.0926 -2.7657 0>}
 ]
}
constraints [
 {
  bone1Name=head
  bone2Name=neck_01
  swing1LimitDegrees=45
  swing1Motion=Free
  swing2LimitDegrees=45
  swing2Motion=Free
  twistLimitDegrees=45
  twistMotion=Free
 }
 {bone1Name=dyn_hair_l_01, bone2Name=dyn_hair_l_MSTR, swing1LimitDegrees=35, swing2LimitDegrees=35, twistLimitDegrees=5}
 {bone1Name=dyn_hair_l_02, bone2Name=dyn_hair_l_01, swing1LimitDegrees=25, swing2LimitDegrees=25, twistLimitDegrees=5}
 {bone1Name=dyn_hair_l_03, bone2Name=dyn_hair_l_02, swing1LimitDegrees=30, swing2LimitDegrees=30, twistLimitDegrees=5}
 {bone1Name=dyn_hair_l_04, bone2Name=dyn_hair_l_03, swing1LimitDegrees=40, swing2LimitDegrees=40, twistLimitDegrees=5}
 {bone1Name=dyn_hair_l_05, bone2Name=dyn_hair_l_04, swing1LimitDegrees=45, swing2LimitDegrees=45, twistLimitDegrees=5}
 {bone1Name=dyn_hair_fl_01, bone2Name=dyn_hair_fl_MSTR, swing1LimitDegrees=20, swing2LimitDegrees=20, twistLimitDegrees=5}
 {bone1Name=dyn_hair_fl_02, bone2Name=dyn_hair_fl_01, swing1LimitDegrees=25, swing2LimitDegrees=25, twistLimitDegrees=5}
 {bone1Name=dyn_hair_fl_03, bone2Name=dyn_hair_fl_02, swing1LimitDegrees=30, swing2LimitDegrees=30, twistLimitDegrees=5}
 {bone1Name=dyn_hair_fl_04, bone2Name=dyn_hair_fl_03, swing1LimitDegrees=40, swing2LimitDegrees=40, twistLimitDegrees=5}
 {bone1Name=dyn_hair_fl_05, bone2Name=dyn_hair_fl_04, swing1LimitDegrees=45, swing2LimitDegrees=45, twistLimitDegrees=5}
 {bone1Name=dyn_hair_bl_01, bone2Name=head, swing1LimitDegrees=20, swing2LimitDegrees=20, twistLimitDegrees=5}
 {bone1Name=dyn_hair_bl_02, bone2Name=dyn_hair_bl_01, swing1LimitDegrees=25, swing2LimitDegrees=25, twistLimitDegrees=5}
 {bone1Name=dyn_hair_bl_03, bone2Name=dyn_hair_bl_02, swing1LimitDegrees=30, swing2LimitDegrees=30, twistLimitDegrees=5}
 {bone1Name=dyn_hair_bl_04, bone2Name=dyn_hair_bl_03, swing1LimitDegrees=40, swing2LimitDegrees=40, twistLimitDegrees=5}
 {bone1Name=dyn_hair_b_01, bone2Name=dyn_hair_b_MSTR, swing1LimitDegrees=20, swing2LimitDegrees=20, twistLimitDegrees=5}
 {bone1Name=dyn_hair_b_02, bone2Name=dyn_hair_b_01, swing1LimitDegrees=20, swing2LimitDegrees=20, twistLimitDegrees=5}
 {bone1Name=dyn_hair_b_03, bone2Name=dyn_hair_b_02, swing1LimitDegrees=25, swing2LimitDegrees=25, twistLimitDegrees=5}
 {bone1Name=dyn_hair_b_04, bone2Name=dyn_hair_b_03, swing1LimitDegrees=40, swing2LimitDegrees=40, twistLimitDegrees=5}
 {bone1Name=dyn_hair_b_05, bone2Name=dyn_hair_b_04, swing1LimitDegrees=30, swing2LimitDegrees=30, twistLimitDegrees=5}
 {bone1Name=dyn_hair_r_01, bone2Name=dyn_hair_r_MSTR, swing1LimitDegrees=20, swing2LimitDegrees=20, twistLimitDegrees=5}
 {bone1Name=dyn_hair_r_02, bone2Name=dyn_hair_r_01, swing1LimitDegrees=30, swing2LimitDegrees=30, twistLimitDegrees=5}
 {bone1Name=dyn_hair_r_03, bone2Name=dyn_hair_r_02, swing1LimitDegrees=15, swing2LimitDegrees=15, twistLimitDegrees=5}
 {bone1Name=dyn_hair_r_04, bone2Name=dyn_hair_r_03, swing1LimitDegrees=15, swing2LimitDegrees=15, twistLimitDegrees=5}
 {bone1Name=dyn_hair_r_05, bone2Name=dyn_hair_r_04, swing1LimitDegrees=15, swing2LimitDegrees=15, twistLimitDegrees=5}
 {bone1Name=dyn_hair_bang_01, bone2Name=head, swing1LimitDegrees=25, swing2LimitDegrees=25, twistLimitDegrees=5}
 {bone1Name=dyn_hair_bang_02, bone2Name=dyn_hair_bang_01, swing1LimitDegrees=15, swing2LimitDegrees=15, twistLimitDegrees=5}
 {bone1Name=dyn_hair_bang_03, bone2Name=dyn_hair_bang_02, swing1LimitDegrees=25, swing2LimitDegrees=25, twistLimitDegrees=5}
 {bone1Name=dyn_hair_f_01, bone2Name=head, swing1LimitDegrees=45, swing2LimitDegrees=45, twistLimitDegrees=15}
 {
  bone1Name=dyn_hair_b_02
  bone2Name=dyn_hair_bl_02
  swing1LimitDegrees=45
  swing1Motion=Free
  swing2LimitDegrees=45
  swing2Motion=Free
  twistLimitDegrees=45
  twistMotion=Free
 }
 {bone1Name=dyn_hair_bm_01, bone2Name=dyn_hair_bm_MSTR, swing1LimitDegrees=20, swing2LimitDegrees=20, twistLimitDegrees=0}
 {bone1Name=dyn_hair_bm_02, bone2Name=dyn_hair_bm_01, swing1LimitDegrees=20, swing2LimitDegrees=20, twistLimitDegrees=0}
 {bone1Name=dyn_hair_bm_03, bone2Name=dyn_hair_bm_02, swing1LimitDegrees=25, swing2LimitDegrees=25, twistLimitDegrees=0}
 {bone1Name=dyn_hair_bm_04, bone2Name=dyn_hair_bm_03, swing1LimitDegrees=30, swing2LimitDegrees=30, twistLimitDegrees=0}
 {bone1Name=dyn_hair_bm_05, bone2Name=dyn_hair_bm_04, swing1LimitDegrees=30, swing2LimitDegrees=30, twistLimitDegrees=0}
 {
  bone1Name=upperarm_r
  bone2Name=clavicle_r
  swing1LimitDegrees=45
  swing1Motion=Free
  swing2LimitDegrees=45
  swing2Motion=Free
  twistLimitDegrees=45
  twistMotion=Free
 }
 {
  bone1Name=dyn_hair_bm_MSTR
  bone2Name=head
  swing1LimitDegrees=45
  swing1Motion=Free
  swing2LimitDegrees=45
  swing2Motion=Free
  twistLimitDegrees=45
  twistMotion=Free
 }
 {
  bone1Name=dyn_hair_r_MSTR
  bone2Name=head
  swing1LimitDegrees=45
  swing1Motion=Free
  swing2LimitDegrees=45
  swing2Motion=Free
  twistLimitDegrees=45
  twistMotion=Free
 }
 {
  bone1Name=dyn_hair_b_MSTR
  bone2Name=head
  swing1LimitDegrees=45
  swing1Motion=Free
  swing2LimitDegrees=45
  swing2Motion=Free
  twistLimitDegrees=45
  twistMotion=Free
 }
 {bone1Name=dyn_hair_blm_04, bone2Name=dyn_hair_blm_03, swing1LimitDegrees=40, swing2LimitDegrees=40, twistLimitDegrees=5}
 {
  bone1Name=dyn_hair_blm_MSTR
  bone2Name=head
  swing1LimitDegrees=45
  swing1Motion=Free
  swing2LimitDegrees=45
  swing2Motion=Free
  twistLimitDegrees=45
  twistMotion=Free
 }
 {bone1Name=dyn_hair_blm_01, bone2Name=dyn_hair_blm_MSTR, swing1LimitDegrees=20, swing2LimitDegrees=20, twistLimitDegrees=5}
 {bone1Name=dyn_hair_blm_02, bone2Name=dyn_hair_blm_01, swing1LimitDegrees=25, swing2LimitDegrees=25, twistLimitDegrees=5}
 {bone1Name=dyn_hair_blm_03, bone2Name=dyn_hair_blm_02, swing1LimitDegrees=30, swing2LimitDegrees=30, twistLimitDegrees=5}
 {
  bone1Name=dyn_hair_blm_03
  bone2Name=dyn_hair_b_04
  swing1LimitDegrees=45
  swing1Motion=Free
  swing2LimitDegrees=45
  swing2Motion=Free
  twistLimitDegrees=45
  twistMotion=Free
 }
 {
  bone1Name=dyn_hair_b_05
  bone2Name=dyn_hair_bm_05
  swing1LimitDegrees=45
  swing1Motion=Free
  swing2LimitDegrees=45
  swing2Motion=Free
  twistLimitDegrees=45
  twistMotion=Free
 }
 {
  bone1Name=dyn_hair_l_MSTR
  bone2Name=head
  swing1LimitDegrees=45
  swing1Motion=Free
  swing2LimitDegrees=45
  swing2Motion=Free
  twistLimitDegrees=45
  twistMotion=Free
 }
 {
  bone1Name=dyn_hair_fl_MSTR
  bone2Name=head
  swing1LimitDegrees=45
  swing1Motion=Free
  swing2LimitDegrees=45
  swing2Motion=Free
  twistLimitDegrees=45
  twistMotion=Free
 }
 {
  bone1Name=dyn_hair_bm_03
  bone2Name=dyn_hair_b_03
  swing1LimitDegrees=45
  swing1Motion=Free
  swing2LimitDegrees=45
  swing2Motion=Free
  twistLimitDegrees=45
  twistMotion=Free
 }
 {
  bone1Name=dyn_hair_r_04
  bone2Name=dyn_hair_bm_04
  swing1LimitDegrees=45
  swing1Motion=Free
  swing2LimitDegrees=45
  swing2Motion=Free
  twistLimitDegrees=45
  twistMotion=Free
 }
]
