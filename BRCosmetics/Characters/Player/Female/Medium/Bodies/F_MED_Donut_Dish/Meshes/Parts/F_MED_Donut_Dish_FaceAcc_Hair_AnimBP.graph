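# Animation Blueprint for the Donut Dish female face-accessory hair mesh, adding physics-driven hair motion on top of the character's pose.
# Copies the attached parent's pose, layers the owner mesh's pose from upperarm_r, then feeds a partial-alpha RigidBody simulation.
# The simulation runs in spine_05 space with the hair physics asset, damped component motion and skydiving external force.
# A targeting-gated chain of additive hair-bone offsets exists but is disconnected from the output.
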AnimBlueprint /BRCosmetics/Characters/Player/Female/Medium/Bodies/F_MED_Donut_Dish/Meshes/Parts/F_MED_Donut_Dish_FaceAcc_Hair_AnimBP
target_skeleton=&/F_MED_Donut_Dish_FaceAcc_Hair_Skeleton
graphs [
 AnimGraph {nodes [
   n1 = Get bIsTargeting()
   n2 = "Transform (Modify) Bone - Bone: dyn_hair_l_MSTR"!(ComponentPose, bAlphaBoolEnabled:bool=@n1, Translation="0,0,0", Rotation="8.040,-2.355,11.717", Scale="1,1,1") -> (Pose) {
    alphaInputType=Bool
    rotationMode=BMM_Additive
    alphaBoolBlend {blendOption=Linear}
    boneToModify {boneName=dyn_hair_l_MSTR}
    rotation=<rot 8.04 -2.355 11.7171>
   }
   n3 = "Transform (Modify) Bone - Bone: dyn_hair_fl_MSTR"!(ComponentPose=@n2, bAlphaBoolEnabled:bool=@n1, Translation="0,0,0", Rotation="10,0,40", Scale="1,1,1") -> (Pose) {
    alphaInputType=Bool
    rotationMode=BMM_Additive
    alphaBoolBlend {blendOption=Linear}
    boneToModify {boneName=dyn_hair_fl_MSTR}
    rotation {pitch=10, roll=40}
   }
   n4 = "Transform (Modify) Bone - Bone: dyn_hair_bl_MSTR"!(ComponentPose=@n3, bAlphaBoolEnabled:bool=@n1, Translation="0,0,0", Rotation="0.000,0.000,7", Scale="1,1,1") -> (Pose) {
    alphaInputType=Bool
    rotationMode=BMM_Additive
    alphaBoolBlend {blendOption=Linear}
    boneToModify {boneName=dyn_hair_bl_MSTR}
    rotation {roll=7}
   }
   n5 = "Transform (Modify) Bone - Bone: dyn_hair_blm_MSTR"!(ComponentPose=@n4, bAlphaBoolEnabled:bool=@n1, Translation="0,0,0", Rotation="0,0,10", Scale="0.1,0.1,0.1") -> (Pose) {
    alphaInputType=Bool
    rotationMode=BMM_Additive
    alphaBoolBlend {blendOption=Linear}
    boneToModify {boneName=dyn_hair_blm_MSTR}
    rotation {roll=10}
    scale=<0.1 0.1 0.1>
   }
   n6 = "Transform (Modify) Bone - Bone: dyn_hair_b_MSTR"!(ComponentPose=@n5, bAlphaBoolEnabled:bool=@n1, Translation="0,0,0", Rotation="2,0.000,10.000", Scale="1,1,1") -> (Pose) {
    alphaInputType=Bool
    rotationMode=BMM_Additive
    alphaBoolBlend {blendOption=Linear}
    boneToModify {boneName=dyn_hair_b_MSTR}
    rotation {pitch=2, roll=10}
   }
   n7 = "Transform (Modify) Bone - Bone: dyn_hair_bm_MSTR"!(ComponentPose=@n6, bAlphaBoolEnabled:bool=@n1, Translation="0,0,0", Rotation="8.052,-0.000,20.000", Scale="1,1,1") -> (Pose) {
    alphaInputType=Bool
    rotationMode=BMM_Additive
    alphaBoolBlend {blendOption=Linear}
    boneToModify {boneName=dyn_hair_bm_MSTR}
    rotation=<rot 8.0522 0 20>
   }
   "Transform (Modify) Bone - Bone: dyn_hair_r_MSTR"!(ComponentPose=@n7, bAlphaBoolEnabled:bool=@n1, Translation="0,0,0", Rotation="5,0.000,30", Scale="0.1,0.1,0.1") -> (Pose) {
    alphaInputType=Bool
    rotationMode=BMM_Additive
    scaleMode=BMM_Replace
    alphaBoolBlend {blendOption=Linear}
    boneToModify {boneName=dyn_hair_r_MSTR}
    rotation {pitch=5, roll=30}
    scale=<0.1 0.1 0.1>
   }
   n8 = Copy Pose From Mesh(SourceMeshComponent=Get Body_Mesh())
   n9 = Copy Pose From Mesh() {bUseAttachedParent=true}
   n10 = RigidBody(
    ComponentPose=Local To Component(LocalPose=Layered blend per bone(BasePose=@n9, BlendPoses_0=@n8) {layerSetup [{branchFilters [{boneName=upperarm_r}]}]})
    Alpha=0.55
    ExternalForce=Break Fort Anim Input Skydiving External Force(FortAnimInput_SkydivingExternalForce=Get SkydivingExternalForceInput()) -> (
     bUseSkydivingVectorForce=~
     SkydivingVectorForce=~
     HeadToPelvisDirection=~
     FloatingMultiplier=~
     FloatingAdditive=~
     DivingMultiplier=~
     DivingAdditive=~
     ParachutingMultiplier=~
     ParachutingAdditive=~
    ).SkydivingVectorForce
   ) {
    bClampLinearTranslationLimitToRefPose=true
    bOverrideWorldGravity=true
    overridePhysicsAsset=&/F_MED_Donut_Dish_FaceAcc_Hair_Physics
    simulationSpace=BaseBoneSpace
    alphaBoolBlend {blendOption=Linear}
    baseBoneRef {boneName=spine_05}
    componentLinearAccScale=<0.5 0.4 0.4>
    componentLinearVelScale=<0.4 0.4 0.4>
    overrideWorldGravity {z=-980}
   }
   Output Pose(Result=Component To Local(ComponentPose=@n10)) {layerGroup=None}
  ]}
 EventGraph {nodes [
   Event BlueprintInitializeAnimation()
   > Set Body_Mesh(Body_Mesh=GetOwnerSkeletalMeshComponent())
   TryGetPawnOwner()
   Event BlueprintUpdateAnimation()
  ]}
]
variables [Body_Mesh {type=SkeletalMeshComponent}]
